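; Artemis boat spline actor that places and enables a dynamic boat foundation at a random point along a spline.
; On BeginPlay, the server waits for playlist data, validates the spline, then sets the foundation's transform and enables it.
; Location snaps to the Fort building grid, and yaw follows spline direction snapped to 90 degrees via GetGridRotationOnSpline, which also debug-prints yaw.
; Location and rotation use separate random index expressions, so their points may differ.
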(event EventBeginPlay
  (bind _boatfoundation (Variables|Default|GetBoatFoundation))
  (bind _boatspline (Variables|Default|GetBoatSpline))
  (Networking|SwitchHasAuthority
    (:Authority
      (Class|FortAsyncActionCurrentPlaylistReady|PlaylistDataReadyAsync
        (:OnReady
          (Utilities|IsValid _boatspline
            (:Is%20Valid
              (Class|BuildingFoundation|SetDynamicFoundationTransform _boatfoundation (Math|Transform|MakeTransform (Fortnite|SnaptoFortBuildingGrid (Spline|GetLocationatSplinePoint _boatspline (Math|Random|RandomIntegerinRange 0 (- (Spline|GetNumberOfSplinePoints (Variables|Default|GetBoatSpline)) 1)) "World")) (Default|GetGridRotationonSpline 0.0 _boatspline (Math|Random|RandomIntegerinRange 0 (- (Spline|GetNumberOfSplinePoints (Variables|Default|GetBoatSpline)) 1)))))
              (Class|BuildingFoundation|SetDynamicFoundationEnabled _boatfoundation true))))))))

(fn GetGridRotationOnSpline
  (params (ForwardVectorYawRotation :type Float%20%28double-precision%29) (Spline :type Spline%20Component%20Object%20Reference) (SplinePoint :type Integer))
  (bind _returnvalue (Math|Rotator|MakeRotator 0.0 0.0 ForwardVectorYawRotation))
  (bind _returnvalue_2 (Spline|GetDirectionatSplinePoint Spline SplinePoint "World"))
  (bind _returnvalue_3 (Math|Vector|RotateVector _returnvalue_2 _returnvalue))
  (bind _returnvalue_4 (Math|Conversions|RotationFromXVector _returnvalue_3))
  (bind _yaw (.yaw _returnvalue_4))
  (Development|PrintString (Utilities|String|ToString(Float) _yaw))
  (return (Math|Rotator|MakeRotator 0.0 0.0 (Math|Float|SnapToGrid(Float) _yaw 9e+01)) _yaw))

(fn ConstructionScript)
